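; Dynamic UI director for The Hive POI that controls the bug biome scene and default POI-visited HUD element.
; It follows view target changes, listening to each target's stateful enter/exit event to add the scene when inside with high boss-manager threat.
; Otherwise an active scene is removed after a delay timer, or immediately on view target change.
; Map discoverability visits toggle the default visited display for overridden POI tags, with delta gates filtering redundant updates.
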(event EventBeginPlay
  ; Begin/EndPlay - Bind Events
  (POIManagement|BindtoPOIEvents)
  (bind _asfort_player_controller (Utilities|Casting|CastToFortPlayerController (DynamicUIDirector|GetOwningLocalPlayerController)))
  (EventDispatchers|BindEventtoOnViewTargetChanged _asfort_player_controller (EventDispatchers|CreateEvent :!event Custom|OnViewTargetChanged))
  (CallFunction|OnViewTargetChanged 0 0 (Pawn|GetViewTarget _asfort_player_controller)))

(event EventEndPlay
  (params (EndPlayReason :type EEndPlayReason%20Enum :default "Destroyed"))
  ; Begin/EndPlay - Bind Events
  (POIManagement|UnbindtoPOIEvents)
  (CallFunction|UnbindListenEvent)
  (bind _asfort_player_controller (Utilities|Casting|CastToFortPlayerController (DynamicUIDirector|GetOwningLocalPlayerController)))
  (EventDispatchers|UnbindEventfromOnViewTargetChanged _asfort_player_controller (EventDispatchers|CreateEvent :!event Custom|OnViewTargetChanged)))

(event Custom|OnViewTargetChanged
  (params (PlayerController :type Fort%20Player%20Controller%20Object%20Reference) (OldViewTarget :type Actor%20Object%20Reference) (NewViewTarget :type Actor%20Object%20Reference))
  ; Listen for enter/exit events
  (CallFunction|UnbindListenEvent)
  (CallFunction|SetBiomeSceneActive false) ; TODO: handle differently as this can be\nspammed in spectate
  (bind (_asynctaskproxy _payload) (GameplayEventRouter|StartListeningToStatefulEvent (GameplayEventRouter|GetGlobalEventRouter) "/TidalCrane_Swarm_Boss/Blueprints/STRUCT_Event_BugBiome_PlayerEnteredExited.STRUCT_Event_BugBiome_PlayerEnteredExited" NewViewTarget)
    (:then
      (Variables|Default|SetEnterExitBugBiomeEvent _asynctaskproxy))
    (:OnEventReceived
      _)
    (:OnSavedEventStateExists
      _))
  (if (and (Utilities|Struct|BreakSTRUCTEventBugBiomePlayerEnteredExited _payload) (Class|BPTidalCraneBossSpawnManager|IsThreatHigh (Utilities|ResolveSoftReference (.Manager_6_733642304E316E4615BDE7B4F3908BCB (Utilities|Struct|BreakSTRUCTEventBugBiomePlayerEnteredExited _payload)))))
    (Utilities|Time|ClearandInvalidateTimerbyHandle (Variables|Default|GetLeaveEventTimerHandle))
    (CallFunction|SetBiomeSceneActive true)
    (elif (Variables|Default|GetIsBugBiomeSceneActive)
      (bind _returnvalue (Utilities|Time|SetTimerbyEvent (delegate Custom|OnLeaveTimerFinished) (Variables|Default|GetTimeToLeaveEvent)))
      (Variables|Default|SetLeaveEventTimerHandle _returnvalue))))

(event Custom|OnLeaveTimerFinished
  ; Listen for enter/exit events
  (CallFunction|SetBiomeSceneActive false))

(event Custom|UnbindListenEvent
  (bind _enterexitbugbiomeevent (Variables|Default|GetEnterExitBugBiomeEvent))
  (Utilities|IsValid _enterexitbugbiomeevent
    (:Is%20Valid
      (GameplayEventRouter|StopListeningtoStatefulEvent _enterexitbugbiomeevent)
      (Variables|Default|SetEnterExitBugBiomeEvent 0))))

(fn BindToPOIEvents
  (bind _output_get (Variables|Default|SetDiscoverabilityComponent (Actor|GetComponentbyClass (DynamicUIDirector|GetOwningLocalPlayerController) "/Script/FortniteGame.FortControllerComponent_MapDiscoverability")))
  (Utilities|IsValid _output_get
    (:Is%20Valid
      (EventDispatchers|BindEventtoOnVisited _output_get (EventDispatchers|CreateEvent :!event OnPOIVisited)))))

(fn OnPOIVisited
  (params (PoiTag :type Gameplay%20Tag%20Structure%20%28by%20ref%29))
  (POIManagement|SetShowDefaultPOI (not (Utilities|Array|ContainsItem (Variables|POIManagement|GetOverriddenPOITags) PoiTag))))

(fn SetBiomeSceneActive
  (params (IsActive :type Boolean))
  (bind _bugbiomescene (Variables|Default|GetBugBiomeScene))
  (bind _updated_variable (Utilities|FlowControl|DeltaGateWildcard (Variables|Default|GetIsBugBiomeSceneActive) IsActive)
    (:Success
      (if _updated_variable
        (DynamicUIDirector|AddScene _bugbiomescene)
        (else
          (DynamicUIDirector|RemoveScene _bugbiomescene))))))

(fn SetShowDefaultPOI
  (params (ShouldShow :type Boolean))
  (Utilities|FlowControl|DeltaGateWildcard (Variables|POIManagement|GetIsShowingDefaultPOI) ShouldShow
    (:Success
      (bind _returnvalue (Class|FortHUDContext|GetFortHUDContextbyPlayer (DynamicUIDirector|GetOwningLocalPlayer)))
      (HUD|SetHUDElementVisibility _returnvalue (GameplayTags|MakeLiteralGameplayTagContainer "(GameplayTags=((TagName=\"HUD.Athena.PlayerVisited\")))") (not (Variables|POIManagement|GetIsShowingDefaultPOI)) "ReplaceVisitedUI"))))

(fn UnbindToPOIEvents
  (bind _discoverabilitycomponent (Variables|Default|GetDiscoverabilityComponent)
    (:then
      (EventDispatchers|UnbindEventfromOnVisited _discoverabilitycomponent (EventDispatchers|CreateEvent :!event OnPOIVisited)))
    (:else)))

(fn ConstructionScript)
