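# Translucent UI material for the Hive POI enter backer, named as an unused reference, drawing a procedural sign from SDF masks.
# A hexagon, stretched through an aspect-ratio "Tuneable 9Slice" UV subgraph, loses a trapezoid and side regions to form the backer.
# Its outline gets middle, top and bottom cutouts, while mirrored, scaled-to-fit boxes form accent dots.
# Lerps layer colors by named masks, and weighted masks combine by max, scaled by Master Opacity.
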Material /TheHivePOIUI/UI/Materials/M_HivePOIEnter_Backer_UnusedReference
graph {
 expressions {
  MaterialExpressionAbs_0 {class=MaterialExpressionAbs, inputs [None {from=MaterialExpressionConstantBiasScale_0, output=""}], properties {}}
  MaterialExpressionAbs_1 {class=MaterialExpressionAbs, inputs [None {from=MaterialExpressionConstantBiasScale_2, output=""}], properties {}}
  MaterialExpressionAdd_0 {
   class=MaterialExpressionAdd
   inputs [
    A {from=MaterialExpressionMaterialFunctionCall_28, output=Normalized UV}
    B {from=MaterialExpressionConstant2Vector_2}
   ]
   properties {constB=0}
  }
  MaterialExpressionAdd_2 {
   class=MaterialExpressionAdd
   inputs [
    A {from=MaterialExpressionReroute_4}
    B {from=MaterialExpressionConstant2Vector_5}
   ]
  }
  MaterialExpressionAdd_3 {
   class=MaterialExpressionAdd
   inputs [
    A {from=MaterialExpressionMaterialFunctionCall_31, output=Scaled to Fit}
    B {from=MaterialExpressionConstant2Vector_11}
   ]
   properties {desc=Offset}
  }
  MaterialExpressionAdd_8 {
   class=MaterialExpressionAdd
   inputs [
    A {from=MaterialExpressionReroute_5}
    B {from=MaterialExpressionConstant2Vector_2}
   ]
  }
  MaterialExpressionAppendVector_0 {
   class=MaterialExpressionAppendVector
   inputs [
    A {from=MaterialExpressionDivide_0, output=""}
    B {from=MaterialExpressionDivide_1, output=""}
   ]
  }
  MaterialExpressionClamp_0 {
   class=MaterialExpressionClamp
   inputs [
    None {from=MaterialExpressionReroute_14, output=UV In}
    Min {from="", output=""}
    Max {from="", output=""}
   ]
   properties {desc=Prevent rounding errors on edges, maxDefault=1}
  }
  MaterialExpressionComponentMask_0 {
   class=MaterialExpressionComponentMask
   inputs [None {from=MaterialExpressionReroute_12, output="Size/Aspect Ratio (v2)"}]
   properties {r=true}
  }
  MaterialExpressionComponentMask_1 {
   class=MaterialExpressionComponentMask
   inputs [None {from=MaterialExpressionReroute_13, output="Size/Aspect Ratio (v2)"}]
   properties {g=true}
  }
  MaterialExpressionComponentMask_2 {class=MaterialExpressionComponentMask, inputs [None {from=MaterialExpressionConstantBiasScale_0}], properties {g=true}}
  MaterialExpressionComponentMask_8 {class=MaterialExpressionComponentMask, inputs [None {from=MaterialExpressionAbs_1}], properties {g=true}}
  MaterialExpressionComposite_0 {
   class=MaterialExpressionComposite
   inputs [
    "Size/Aspect Ratio (v2)" {from=MaterialExpressionConstant2Vector_4, output=""}
    "Size/Aspect Ratio (v2)" {from=MaterialExpressionConstant2Vector_4, output=""}
    UV In {from=MaterialExpressionMaterialFunctionCall_4, output=Rotated Values}
   ]
   properties {subgraphName=Tuneable 9Slice}
  }
  MaterialExpressionConstant2Vector_11 {class=MaterialExpressionConstant2Vector, properties {g=-0.22, r=-0.19}}
  MaterialExpressionConstant2Vector_2 {class=MaterialExpressionConstant2Vector, properties {g=-0.01}}
  MaterialExpressionConstant2Vector_4 {class=MaterialExpressionConstant2Vector, properties {g=1, r=3.1}}
  MaterialExpressionConstant2Vector_5 {class=MaterialExpressionConstant2Vector, properties {g=-0.3}}
  MaterialExpressionConstant2Vector_7 {class=MaterialExpressionConstant2Vector, properties {g=1}}
  MaterialExpressionConstant2Vector_8 {class=MaterialExpressionConstant2Vector, properties {g=0.075, r=0.08}}
  MaterialExpressionConstant3Vector_0 {class=MaterialExpressionConstant3Vector, properties {constant=<rgba 0.4 0.5 0.25 0>}}
  MaterialExpressionConstant3Vector_6 {class=MaterialExpressionConstant3Vector, properties {constant=<rgba 1 0.2346 0 1>}}
  MaterialExpressionConstantBiasScale_0 {class=MaterialExpressionConstantBiasScale, inputs [None {from=MaterialExpressionReroute_5}], properties {bias=-0.5, scale=2}}
  MaterialExpressionConstantBiasScale_2 {class=MaterialExpressionConstantBiasScale, inputs [None {from=MaterialExpressionAdd_8}], properties {bias=-0.5, scale=2}}
  MaterialExpressionConstant_0 {class=MaterialExpressionConstant, properties {r=6}}
  MaterialExpressionConstant_1 {class=MaterialExpressionConstant}
  MaterialExpressionConstant_21 {class=MaterialExpressionConstant, properties {r=0.01}}
  MaterialExpressionConstant_3 {class=MaterialExpressionConstant, properties {r=0.5}}
  MaterialExpressionConstant_4 {class=MaterialExpressionConstant, properties {r=0.02}}
  MaterialExpressionConstant_5 {class=MaterialExpressionConstant, properties {r=2.75}}
  MaterialExpressionDivide_0 {
   class=MaterialExpressionDivide
   inputs [
    A {from=MaterialExpressionComponentMask_0}
    B {from=MaterialExpressionComponentMask_1}
   ]
  }
  MaterialExpressionDivide_1 {
   class=MaterialExpressionDivide
   inputs [
    A {from=MaterialExpressionComponentMask_1}
    B {from=MaterialExpressionComponentMask_0}
   ]
  }
  MaterialExpressionFloor_0 {class=MaterialExpressionFloor, inputs [None {from=MaterialExpressionMultiply_5, output=""}], properties {}}
  MaterialExpressionLinearInterpolate_1 {
   class=MaterialExpressionLinearInterpolate
   inputs [
    A {from=MaterialExpressionReroute_0}
    B {from=MaterialExpressionMax_1}
    Alpha {from=MaterialExpressionReroute_1}
   ]
  }
  MaterialExpressionLinearInterpolate_13 {
   class=MaterialExpressionLinearInterpolate
   inputs [
    A {from=""}
    B {from=MaterialExpressionConstant3Vector_6}
    Alpha {from=MaterialExpressionNamedRerouteUsage_20}
   ]
  }
  MaterialExpressionLinearInterpolate_14 {
   class=MaterialExpressionLinearInterpolate
   inputs [
    A {from=MaterialExpressionLinearInterpolate_13}
    B {from=""}
    Alpha {from=MaterialExpressionNamedRerouteUsage_19}
   ]
   properties {constB=0.5}
  }
  MaterialExpressionLinearInterpolate_18 {
   class=MaterialExpressionLinearInterpolate
   inputs [
    A {from=MaterialExpressionLinearInterpolate_14}
    B {from=MaterialExpressionConstant_21}
    Alpha {from=MaterialExpressionNamedRerouteUsage_18}
   ]
  }
  MaterialExpressionMaterialFunctionCall_13 {
   class=MaterialExpressionMaterialFunctionCall
   inputs [
    UVs {from=MaterialExpressionAdd_2}
    TopWidth {from=MaterialExpressionConstant3Vector_0, output=R}
    BottomWidth {from=MaterialExpressionConstant3Vector_0, output=R}
    Height {from=MaterialExpressionConstant3Vector_0, output=R}
   ]
   properties {materialFunction=/FortUILibrary/MaterialFunctions/Shapes/SDFs/MF_SDF_IsoscelesTrapezoid}
  }
  MaterialExpressionMaterialFunctionCall_14 {
   class=MaterialExpressionMaterialFunctionCall
   inputs [
    DistanceFieldA {from=MaterialExpressionMaterialFunctionCall_3, output=SDF}
    DistanceFieldB {from=MaterialExpressionMaterialFunctionCall_13, output=SDF}
    BlendAmount {from=""}
   ]
   properties {materialFunction=/FortUILibrary/MaterialFunctions/Shapes/Modifiers/MF_UI_SDF_Subtraction}
  }
  MaterialExpressionMaterialFunctionCall_15 {
   class=MaterialExpressionMaterialFunctionCall
   inputs [Float2 {from=MaterialExpressionAbs_0}]
   properties {materialFunction=&BreakOutFloat2Components}
  }
  MaterialExpressionMaterialFunctionCall_16 {
   class=MaterialExpressionMaterialFunctionCall
   inputs [
    DistanceFieldA {from=MaterialExpressionMaterialFunctionCall_14, output=Exact}
    DistanceFieldB {from=MaterialExpressionMultiply_3}
    BlendAmount {from=""}
   ]
   properties {materialFunction=&MF_UI_SDF_Subtraction}
  }
  MaterialExpressionMaterialFunctionCall_18 {
   class=MaterialExpressionMaterialFunctionCall
   inputs [
    DistanceField {from=MaterialExpressionMaterialFunctionCall_16, output=Exact}
    Threshold {from=""}
    Smoothing {from=""}
   ]
   properties {materialFunction=&MF_UI_AntiAliasing}
  }
  MaterialExpressionMaterialFunctionCall_19 {
   class=MaterialExpressionMaterialFunctionCall
   inputs [
    DistanceField {from=MaterialExpressionMaterialFunctionCall_16, output=Exact}
    Threshold {from=""}
    Thickness {from=MaterialExpressionConstant_4}
    Alignment {from=MaterialExpressionConstant_5}
    Smoothing {from=""}
   ]
   properties {materialFunction=/FortUILibrary/MaterialFunctions/Shapes/Modifiers/MF_UI_ShapeOutline}
  }
  MaterialExpressionMaterialFunctionCall_20 {
   class=MaterialExpressionMaterialFunctionCall
   inputs [
    DistanceFieldA {from=MaterialExpressionMaterialFunctionCall_19, output=SDFOutline}
    DistanceFieldB {from=MaterialExpressionSubtract_3}
    BlendAmount {from=""}
   ]
   properties {materialFunction=&MF_UI_SDF_Subtraction}
  }
  MaterialExpressionMaterialFunctionCall_21 {
   class=MaterialExpressionMaterialFunctionCall
   inputs [
    DistanceField {from=MaterialExpressionMaterialFunctionCall_26, output=Exact}
    Threshold {from=""}
    Smoothing {from=""}
   ]
   properties {materialFunction=&MF_UI_AntiAliasing}
  }
  MaterialExpressionMaterialFunctionCall_26 {
   class=MaterialExpressionMaterialFunctionCall
   inputs [
    DistanceFieldA {from=MaterialExpressionMaterialFunctionCall_20, output=Exact}
    DistanceFieldB {from=MaterialExpressionMax_4}
    BlendAmount {from=""}
   ]
   properties {materialFunction=/FortUILibrary/MaterialFunctions/Shapes/Modifiers/MF_UI_SDF_Intersection}
  }
  MaterialExpressionMaterialFunctionCall_28 {class=MaterialExpressionMaterialFunctionCall, inputs [], properties {materialFunction=&GetUserInterfaceUV}}
  MaterialExpressionMaterialFunctionCall_3 {
   class=MaterialExpressionMaterialFunctionCall
   inputs [
    UVs {from=MaterialExpressionReroute_15, output=UV Out}
    Radius {from=MaterialExpressionConstant_3}
    Sides {from=MaterialExpressionConstant_0}
    InnerAngle {from=""}
   ]
   properties {materialFunction=/FortUILibrary/MaterialFunctions/Shapes/SDFs/MF_SDF_RegularPolygon}
  }
  MaterialExpressionMaterialFunctionCall_31 {
   class=MaterialExpressionMaterialFunctionCall
   inputs [
    UV {from=MaterialExpressionMaterialFunctionCall_33, output=Symmetry}
    TextureSize {from=""}
    ScalePivotPoint {from=MaterialExpressionConstant2Vector_7}
    ContainerSize {from=MaterialExpressionReroute_11}
   ]
   properties {materialFunction=&MF_UI_UVScaling}
  }
  MaterialExpressionMaterialFunctionCall_32 {
   class=MaterialExpressionMaterialFunctionCall
   inputs [
    UVs {from=MaterialExpressionAdd_3}
    Size {from=MaterialExpressionConstant2Vector_8}
    CornerRoundness {from=""}
    KeepAspectRatio {from=""}
    PixelSize {from=""}
   ]
   properties {materialFunction=/FortUILibrary/MaterialFunctions/Shapes/SDFs/MF_SDF_Box}
  }
  MaterialExpressionMaterialFunctionCall_33 {
   class=MaterialExpressionMaterialFunctionCall
   inputs [
    UV {from=MaterialExpressionAdd_0}
    Swizzle UV {from=""}
   ]
   properties {materialFunction=/FortUILibrary/MaterialFunctions/Basics/MF_UI_Symmetry_2x2_New}
  }
  MaterialExpressionMaterialFunctionCall_34 {
   class=MaterialExpressionMaterialFunctionCall
   inputs [
    DistanceField {from=MaterialExpressionMaterialFunctionCall_32, output=SDF}
    Threshold {from=""}
    Smoothing {from=""}
   ]
   properties {materialFunction=&MF_UI_AntiAliasing}
  }
  MaterialExpressionMaterialFunctionCall_4 {
   class=MaterialExpressionMaterialFunctionCall
   inputs [
    UVs {from=MaterialExpressionMaterialFunctionCall_5, output=Normalized UV}
    Rotation Center {from=""}
    "Rotation Angle (0-1)" {from=MaterialExpressionConstant_1}
   ]
   properties {materialFunction=&CustomRotator}
  }
  MaterialExpressionMaterialFunctionCall_5 {class=MaterialExpressionMaterialFunctionCall, inputs [], properties {materialFunction=&GetUserInterfaceUV}}
  MaterialExpressionMax_0 {
   class=MaterialExpressionMax
   inputs [
    A {from=MaterialExpressionAppendVector_0}
    B {from=""}
   ]
  }
  MaterialExpressionMax_1 {
   class=MaterialExpressionMax
   inputs [
    A {from=MaterialExpressionOneMinus_1}
    B {from=""}
   ]
   properties {constB=0.5}
  }
  MaterialExpressionMax_13 {
   class=MaterialExpressionMax
   inputs [
    A {from=MaterialExpressionMultiply_35}
    B {from=MaterialExpressionMultiply_2}
   ]
  }
  MaterialExpressionMax_15 {
   class=MaterialExpressionMax
   inputs [
    A {from=MaterialExpressionMax_13}
    B {from=MaterialExpressionNamedRerouteUsage_9}
   ]
  }
  MaterialExpressionMax_4 {
   class=MaterialExpressionMax
   inputs [
    A {from=MaterialExpressionMin_3}
    B {from=MaterialExpressionMin_2}
   ]
  }
  MaterialExpressionMin_1 {
   class=MaterialExpressionMin
   inputs [
    A {from=MaterialExpressionMultiply_4, output=""}
    B {from="", output=""}
   ]
   properties {constB=0.5}
  }
  MaterialExpressionMin_2 {
   class=MaterialExpressionMin
   inputs [
    A {from=MaterialExpressionMultiply_12, output=""}
    B {from=MaterialExpressionMultiply_15, output=""}
   ]
   properties {}
  }
  MaterialExpressionMin_3 {
   class=MaterialExpressionMin
   inputs [
    A {from=MaterialExpressionMultiply_16, output=""}
    B {from=MaterialExpressionComponentMask_2, output=""}
   ]
   properties {}
  }
  MaterialExpressionMultiply_1 {
   class=MaterialExpressionMultiply
   inputs [
    A {from=MaterialExpressionMax_0}
    B {from=MaterialExpressionOneMinus_0}
   ]
  }
  MaterialExpressionMultiply_12 {
   class=MaterialExpressionMultiply
   inputs [
    A {from=MaterialExpressionSubtract_5}
    B {from=""}
   ]
   properties {constB=-1}
  }
  MaterialExpressionMultiply_15 {
   class=MaterialExpressionMultiply
   inputs [
    A {from=MaterialExpressionComponentMask_2}
    B {from=""}
   ]
   properties {constB=-1}
  }
  MaterialExpressionMultiply_16 {
   class=MaterialExpressionMultiply
   inputs [
    A {from=MaterialExpressionSubtract_6}
    B {from=""}
   ]
   properties {constB=-1}
  }
  MaterialExpressionMultiply_2 {
   class=MaterialExpressionMultiply
   inputs [
    A {from=MaterialExpressionNamedRerouteUsage_8}
    B {from=""}
   ]
   properties {constB=0.5}
  }
  MaterialExpressionMultiply_3 {
   class=MaterialExpressionMultiply
   inputs [
    A {from=MaterialExpressionSubtract_1}
    B {from=""}
   ]
   properties {constB=-2}
  }
  MaterialExpressionMultiply_35 {
   class=MaterialExpressionMultiply
   inputs [
    A {from=MaterialExpressionNamedRerouteUsage_7}
    B {from=""}
   ]
   properties {constB=0.8}
  }
  MaterialExpressionMultiply_37 {
   class=MaterialExpressionMultiply
   inputs [
    A {from=MaterialExpressionMax_15}
    B {from=MaterialExpressionScalarParameter_0}
   ]
  }
  MaterialExpressionMultiply_4 {
   class=MaterialExpressionMultiply
   inputs [
    A {from=MaterialExpressionMax_0}
    B {from=MaterialExpressionClamp_0}
   ]
  }
  MaterialExpressionMultiply_5 {
   class=MaterialExpressionMultiply
   inputs [
    A {from=MaterialExpressionClamp_0}
    B {from=""}
   ]
   properties {constB=2}
  }
  MaterialExpressionNamedRerouteDeclaration_4 {
   class=MaterialExpressionNamedRerouteDeclaration
   inputs [None {from=MaterialExpressionSaturate_12}]
   properties {name=Mask_SignBacker, nodeColor=<rgba 0.75 0.6063 0.375 0.75>}
  }
  MaterialExpressionNamedRerouteDeclaration_5 {
   class=MaterialExpressionNamedRerouteDeclaration
   inputs [None {from=MaterialExpressionSaturate_5}]
   properties {name=Mask_SignOutline, nodeColor=<rgba 0.75 0.6063 0.375 0.75>}
  }
  MaterialExpressionNamedRerouteDeclaration_6 {
   class=MaterialExpressionNamedRerouteDeclaration
   inputs [None {from=MaterialExpressionSaturate_7}]
   properties {name=Mask_SignAccentDots, nodeColor=<rgba 0.75 0.6063 0.375 0.75>}
  }
  MaterialExpressionNamedRerouteUsage_18 {class=MaterialExpressionNamedRerouteUsage}
  MaterialExpressionNamedRerouteUsage_19 {class=MaterialExpressionNamedRerouteUsage}
  MaterialExpressionNamedRerouteUsage_20 {class=MaterialExpressionNamedRerouteUsage}
  MaterialExpressionNamedRerouteUsage_7 {class=MaterialExpressionNamedRerouteUsage}
  MaterialExpressionNamedRerouteUsage_8 {class=MaterialExpressionNamedRerouteUsage}
  MaterialExpressionNamedRerouteUsage_9 {class=MaterialExpressionNamedRerouteUsage}
  MaterialExpressionOneMinus_0 {class=MaterialExpressionOneMinus, inputs [None {from=MaterialExpressionClamp_0}]}
  MaterialExpressionOneMinus_1 {class=MaterialExpressionOneMinus, inputs [None {from=MaterialExpressionMultiply_1}]}
  MaterialExpressionPinBase_0 {
   class=MaterialExpressionPinBase
   inputs []
   properties {reroutePins [
     "Size/Aspect Ratio (v2)" {expression=&self:MaterialGraph_1.MaterialGraphNode_Composite_1.Tuneable 9Slice.MaterialExpressionReroute_12}
     "Size/Aspect Ratio (v2)" {expression=&self:MaterialGraph_1.MaterialGraphNode_Composite_1.Tuneable 9Slice.MaterialExpressionReroute_13}
     UV In {expression=&self:MaterialGraph_1.MaterialGraphNode_Composite_1.Tuneable 9Slice.MaterialExpressionReroute_14}
    ]}
  }
  MaterialExpressionPinBase_1 {
   class=MaterialExpressionPinBase
   inputs [UV Out {from=MaterialExpressionLinearInterpolate_1, output=""}]
   properties {reroutePins [UV Out {expression=&self:MaterialGraph_1.MaterialGraphNode_Composite_1.Tuneable 9Slice.MaterialExpressionReroute_15}]}
  }
  MaterialExpressionReroute_0 {class=MaterialExpressionReroute, inputs [None {from=MaterialExpressionMin_1}]}
  MaterialExpressionReroute_1 {class=MaterialExpressionReroute, inputs [None {from=MaterialExpressionFloor_0}]}
  MaterialExpressionReroute_11 {class=MaterialExpressionReroute, inputs [None {from=MaterialExpressionMaterialFunctionCall_28, output=Pixel Size}]}
  MaterialExpressionReroute_4 {class=MaterialExpressionReroute, inputs [None {from=MaterialExpressionMaterialFunctionCall_5, output=Normalized UV}]}
  MaterialExpressionReroute_5 {class=MaterialExpressionReroute, inputs [None {from=MaterialExpressionMaterialFunctionCall_5, output=Normalized UV}]}
  MaterialExpressionSaturate_12 {class=MaterialExpressionSaturate, inputs [None {from=MaterialExpressionMaterialFunctionCall_18, output=Inner}]}
  MaterialExpressionSaturate_5 {class=MaterialExpressionSaturate, inputs [None {from=MaterialExpressionMaterialFunctionCall_21, output=Inner}]}
  MaterialExpressionSaturate_7 {class=MaterialExpressionSaturate, inputs [None {from=MaterialExpressionMaterialFunctionCall_34, output=Inner}]}
  MaterialExpressionScalarParameter_0 {class=MaterialExpressionScalarParameter, properties {defaultValue=0.8, parameterName=Master Opacity}}
  MaterialExpressionSubtract_1 {
   class=MaterialExpressionSubtract
   inputs [
    A {from=MaterialExpressionMaterialFunctionCall_15, output=R}
    B {from=""}
   ]
   properties {constB=0.725}
  }
  MaterialExpressionSubtract_3 {
   class=MaterialExpressionSubtract
   inputs [
    A {from=MaterialExpressionComponentMask_8}
    B {from=""}
   ]
   properties {constB=0.35, desc=MiddleCutoutSize}
  }
  MaterialExpressionSubtract_5 {
   class=MaterialExpressionSubtract
   inputs [
    A {from=MaterialExpressionMaterialFunctionCall_15, output=R}
    B {from=""}
   ]
   properties {constB=0.175, desc=TopLineCutoutSize}
  }
  MaterialExpressionSubtract_6 {
   class=MaterialExpressionSubtract
   inputs [
    A {from=MaterialExpressionMaterialFunctionCall_15, output=R}
    B {from=""}
   ]
   properties {constB=0.0625, desc=BottomLineCutoutSize}
  }
 }
 properties {bAutomaticallySetUsageInEditor=false, blendMode=BLEND_Translucent, materialDomain=MD_UI, usedShadingModels=Default Lit}
}
