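# M_HivePOIEnter_FX is a translucent UI material that draws procedural accent effects, apparently for the Hive POI entry display.
# A tuneable 9-slice subgraph stretches hexagon SDF UVs; trapezoid and side subtractions carve it into the backer mask.
# Circle SDFs form a dot grid and soft glow, while rotated sine bands form stripe accents.
# Named reroutes carry these masks into max, multiply and orange-tint blends, without textures or time-based animation.
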Material /TheHivePOIUI/UI/Materials/M_HivePOIEnter_FX
graph {
 expressions {
  MaterialExpressionAbs_3 {class=MaterialExpressionAbs, inputs [None {from=MaterialExpressionConstantBiasScale_4, output=""}], properties {}}
  MaterialExpressionAdd_0 {
   class=MaterialExpressionAdd
   inputs [
    A {from=MaterialExpressionNamedRerouteUsage_2}
    B {from=MaterialExpressionNamedRerouteUsage_3}
   ]
  }
  MaterialExpressionAdd_14 {
   class=MaterialExpressionAdd
   inputs [
    A {from=MaterialExpressionReroute_6}
    B {from=MaterialExpressionConstant2Vector_10}
   ]
  }
  MaterialExpressionAdd_15 {
   class=MaterialExpressionAdd
   inputs [
    A {from=MaterialExpressionMaterialFunctionCall_27, output=Normalized UV}
    B {from=MaterialExpressionConstant2Vector_14}
   ]
  }
  MaterialExpressionAdd_19 {
   class=MaterialExpressionAdd
   inputs [
    A {from=MaterialExpressionMaterialFunctionCall_43, output=Normalized UV}
    B {from=MaterialExpressionConstant2Vector_15}
   ]
  }
  MaterialExpressionAppendVector_0 {
   class=MaterialExpressionAppendVector
   inputs [
    A {from=MaterialExpressionDivide_0, output=""}
    B {from=MaterialExpressionDivide_1, output=""}
   ]
  }
  MaterialExpressionClamp_0 {
   class=MaterialExpressionClamp
   inputs [
    None {from=MaterialExpressionReroute_14, output=UV In}
    Min {from="", output=""}
    Max {from="", output=""}
   ]
   properties {desc=Prevent rounding errors on edges, maxDefault=1}
  }
  MaterialExpressionComponentMask_0 {
   class=MaterialExpressionComponentMask
   inputs [None {from=MaterialExpressionReroute_12, output="Size/Aspect Ratio (v2)"}]
   properties {r=true}
  }
  MaterialExpressionComponentMask_1 {
   class=MaterialExpressionComponentMask
   inputs [None {from=MaterialExpressionReroute_13, output="Size/Aspect Ratio (v2)"}]
   properties {g=true}
  }
  MaterialExpressionComponentMask_10 {class=MaterialExpressionComponentMask, inputs [None {from=MaterialExpressionSubtract_4}], properties {g=true}}
  MaterialExpressionComponentMask_12 {class=MaterialExpressionComponentMask, inputs [None {from=MaterialExpressionSubtract_9}], properties {g=true}}
  MaterialExpressionComponentMask_9 {
   class=MaterialExpressionComponentMask
   inputs [None {from=MaterialExpressionMaterialFunctionCall_11, output=Rotated Values}]
   properties {r=true}
  }
  MaterialExpressionComposite_1 {
   class=MaterialExpressionComposite
   inputs [
    "Size/Aspect Ratio (v2)" {from=MaterialExpressionConstant2Vector_9, output=""}
    "Size/Aspect Ratio (v2)" {from=MaterialExpressionConstant2Vector_9, output=""}
    UV In {from=MaterialExpressionMaterialFunctionCall_25, output=Rotated Values}
   ]
   properties {subgraphName=Tuneable 9Slice}
  }
  MaterialExpressionConstant2Vector_10 {class=MaterialExpressionConstant2Vector, properties {g=-0.3}}
  MaterialExpressionConstant2Vector_14 {class=MaterialExpressionConstant2Vector, properties {g=0.025}}
  MaterialExpressionConstant2Vector_15 {class=MaterialExpressionConstant2Vector, properties {g=-0.2}}
  MaterialExpressionConstant2Vector_6 {class=MaterialExpressionConstant2Vector, properties {g=75, r=210}}
  MaterialExpressionConstant2Vector_9 {class=MaterialExpressionConstant2Vector, properties {g=1.05, r=3}}
  MaterialExpressionConstant3Vector_1 {class=MaterialExpressionConstant3Vector, properties {constant=<rgba 0.4 0.5 0.25 0>}}
  MaterialExpressionConstant3Vector_2 {class=MaterialExpressionConstant3Vector, properties {constant=<rgba 1 0.2346 0 1>}}
  MaterialExpressionConstantBiasScale_4 {class=MaterialExpressionConstantBiasScale, inputs [None {from=MaterialExpressionReroute_7}], properties {bias=-0.5, scale=2}}
  MaterialExpressionConstant_12 {class=MaterialExpressionConstant, properties {r=0.2}}
  MaterialExpressionConstant_13 {class=MaterialExpressionConstant, properties {r=0.225}}
  MaterialExpressionConstant_14 {class=MaterialExpressionConstant, properties {r=6}}
  MaterialExpressionConstant_15 {class=MaterialExpressionConstant}
  MaterialExpressionConstant_17 {class=MaterialExpressionConstant, properties {r=0.5}}
  MaterialExpressionConstant_18 {class=MaterialExpressionConstant, properties {r=-0.2}}
  MaterialExpressionConstant_6 {class=MaterialExpressionConstant, properties {r=0.6}}
  MaterialExpressionConstant_7 {class=MaterialExpressionConstant, properties {r=0.01}}
  MaterialExpressionConstant_8 {class=MaterialExpressionConstant, properties {r=0.5}}
  MaterialExpressionConstant_9 {class=MaterialExpressionConstant, properties {r=-0.14}}
  MaterialExpressionDivide_0 {
   class=MaterialExpressionDivide
   inputs [
    A {from=MaterialExpressionComponentMask_0}
    B {from=MaterialExpressionComponentMask_1}
   ]
  }
  MaterialExpressionDivide_1 {
   class=MaterialExpressionDivide
   inputs [
    A {from=MaterialExpressionComponentMask_1}
    B {from=MaterialExpressionComponentMask_0}
   ]
  }
  MaterialExpressionFloor_0 {class=MaterialExpressionFloor, inputs [None {from=MaterialExpressionMultiply_3, output=""}], properties {}}
  MaterialExpressionFrac_0 {class=MaterialExpressionFrac, inputs [None {from=MaterialExpressionMultiply_0, output=""}], properties {}}
  MaterialExpressionFrac_1 {class=MaterialExpressionFrac, inputs [None {from=MaterialExpressionMultiply_14, output=""}], properties {}}
  MaterialExpressionFrac_2 {class=MaterialExpressionFrac, inputs [None {from=MaterialExpressionMultiply_18, output=""}], properties {}}
  MaterialExpressionLinearInterpolate_0 {
   class=MaterialExpressionLinearInterpolate
   inputs [
    A {from=MaterialExpressionMaterialFunctionCall_7, output=SDF}
    B {from=""}
    Alpha {from=MaterialExpressionMultiply_10}
   ]
  }
  MaterialExpressionLinearInterpolate_1 {
   class=MaterialExpressionLinearInterpolate
   inputs [
    A {from=MaterialExpressionReroute_0}
    B {from=MaterialExpressionMax_1}
    Alpha {from=MaterialExpressionReroute_1}
   ]
  }
  MaterialExpressionLinearInterpolate_2 {
   class=MaterialExpressionLinearInterpolate
   inputs [
    A {from=MaterialExpressionAdd_0}
    B {from=MaterialExpressionConstant3Vector_2}
    Alpha {from=MaterialExpressionNamedRerouteUsage_0}
   ]
  }
  MaterialExpressionMaterialFunctionCall_0 {
   class=MaterialExpressionMaterialFunctionCall
   inputs [Float2 {from=MaterialExpressionOneMinus_4}]
   properties {materialFunction=&BreakOutFloat2Components}
  }
  MaterialExpressionMaterialFunctionCall_11 {
   class=MaterialExpressionMaterialFunctionCall
   inputs [
    UVs {from=MaterialExpressionMaterialFunctionCall_9, output=Normalized UV}
    Rotation Center {from=""}
    "Rotation Angle (0-1)" {from=MaterialExpressionConstant_7}
   ]
   properties {materialFunction=&CustomRotator}
  }
  MaterialExpressionMaterialFunctionCall_12 {
   class=MaterialExpressionMaterialFunctionCall
   inputs [
    DistanceField {from=MaterialExpressionSubtract_2}
    Threshold {from=""}
    Smoothing {from=""}
   ]
   properties {materialFunction=&MF_UI_AntiAliasing}
  }
  MaterialExpressionMaterialFunctionCall_17 {
   class=MaterialExpressionMaterialFunctionCall
   inputs [
    UVs {from=MaterialExpressionMaterialFunctionCall_9, output=Normalized UV}
    Rotation Center {from=""}
    "Rotation Angle (0-1)" {from=MaterialExpressionConstant_9}
   ]
   properties {materialFunction=&CustomRotator}
  }
  MaterialExpressionMaterialFunctionCall_22 {
   class=MaterialExpressionMaterialFunctionCall
   inputs [
    DistanceField {from=MaterialExpressionSine_1}
    Threshold {from=MaterialExpressionConstant_18}
    Smoothing {from=""}
   ]
   properties {materialFunction=&MF_UI_AntiAliasing}
  }
  MaterialExpressionMaterialFunctionCall_24 {
   class=MaterialExpressionMaterialFunctionCall
   inputs [
    UVs {from=MaterialExpressionReroute_15, output=UV Out}
    Radius {from=MaterialExpressionConstant_17}
    Sides {from=MaterialExpressionConstant_14}
    InnerAngle {from=""}
   ]
   properties {materialFunction=/FortUILibrary/MaterialFunctions/Shapes/SDFs/MF_SDF_RegularPolygon}
  }
  MaterialExpressionMaterialFunctionCall_25 {
   class=MaterialExpressionMaterialFunctionCall
   inputs [
    UVs {from=MaterialExpressionAdd_15}
    Rotation Center {from=""}
    "Rotation Angle (0-1)" {from=MaterialExpressionConstant_15}
   ]
   properties {materialFunction=&CustomRotator}
  }
  MaterialExpressionMaterialFunctionCall_27 {class=MaterialExpressionMaterialFunctionCall, inputs [], properties {materialFunction=&GetUserInterfaceUV}}
  MaterialExpressionMaterialFunctionCall_29 {
   class=MaterialExpressionMaterialFunctionCall
   inputs [
    UVs {from=MaterialExpressionAdd_14}
    TopWidth {from=MaterialExpressionConstant3Vector_1, output=R}
    BottomWidth {from=MaterialExpressionConstant3Vector_1, output=R}
    Height {from=MaterialExpressionConstant3Vector_1, output=R}
   ]
   properties {materialFunction=/FortUILibrary/MaterialFunctions/Shapes/SDFs/MF_SDF_IsoscelesTrapezoid}
  }
  MaterialExpressionMaterialFunctionCall_30 {
   class=MaterialExpressionMaterialFunctionCall
   inputs [
    DistanceFieldA {from=MaterialExpressionMaterialFunctionCall_24, output=SDF}
    DistanceFieldB {from=MaterialExpressionMaterialFunctionCall_29, output=SDF}
    BlendAmount {from=""}
   ]
   properties {materialFunction=/FortUILibrary/MaterialFunctions/Shapes/Modifiers/MF_UI_SDF_Subtraction}
  }
  MaterialExpressionMaterialFunctionCall_35 {
   class=MaterialExpressionMaterialFunctionCall
   inputs [Float2 {from=MaterialExpressionAbs_3}]
   properties {materialFunction=&BreakOutFloat2Components}
  }
  MaterialExpressionMaterialFunctionCall_36 {
   class=MaterialExpressionMaterialFunctionCall
   inputs [
    DistanceFieldA {from=MaterialExpressionMaterialFunctionCall_30, output=Exact}
    DistanceFieldB {from=MaterialExpressionMultiply_28}
    BlendAmount {from=""}
   ]
   properties {materialFunction=&MF_UI_SDF_Subtraction}
  }
  MaterialExpressionMaterialFunctionCall_37 {
   class=MaterialExpressionMaterialFunctionCall
   inputs [
    DistanceField {from=MaterialExpressionMaterialFunctionCall_36, output=Exact}
    Threshold {from=""}
    Smoothing {from=""}
   ]
   properties {materialFunction=&MF_UI_AntiAliasing}
  }
  MaterialExpressionMaterialFunctionCall_38 {
   class=MaterialExpressionMaterialFunctionCall
   inputs [
    UVs {from=MaterialExpressionMaterialFunctionCall_42, output=ScaleX}
    Radius {from=""}
   ]
   properties {materialFunction=/FortUILibrary/MaterialFunctions/Shapes/SDFs/MF_SDF_Circle}
  }
  MaterialExpressionMaterialFunctionCall_42 {
   class=MaterialExpressionMaterialFunctionCall
   inputs [
    UV {from=MaterialExpressionAdd_19}
    TextureSize {from=""}
    ScalePivotPoint {from=""}
    ContainerSize {from=""}
   ]
   properties {materialFunction=&MF_UI_UVScaling}
  }
  MaterialExpressionMaterialFunctionCall_43 {class=MaterialExpressionMaterialFunctionCall, inputs [], properties {materialFunction=&GetUserInterfaceUV}}
  MaterialExpressionMaterialFunctionCall_44 {
   class=MaterialExpressionMaterialFunctionCall
   inputs [
    DistanceField {from=MaterialExpressionMultiply_13}
    Threshold {from=""}
    Smoothing {from=""}
   ]
   properties {materialFunction=&MF_UI_AntiAliasing}
  }
  MaterialExpressionMaterialFunctionCall_7 {
   class=MaterialExpressionMaterialFunctionCall
   inputs [
    UVs {from=MaterialExpressionFrac_0}
    Radius {from=MaterialExpressionConstant_6}
   ]
   properties {materialFunction=&MF_SDF_Circle}
  }
  MaterialExpressionMaterialFunctionCall_9 {class=MaterialExpressionMaterialFunctionCall, inputs [], properties {materialFunction=&GetUserInterfaceUV}}
  MaterialExpressionMax_0 {
   class=MaterialExpressionMax
   inputs [
    A {from=MaterialExpressionAppendVector_0}
    B {from=""}
   ]
  }
  MaterialExpressionMax_1 {
   class=MaterialExpressionMax
   inputs [
    A {from=MaterialExpressionOneMinus_1}
    B {from=""}
   ]
   properties {constB=0.5}
  }
  MaterialExpressionMax_2 {
   class=MaterialExpressionMax
   inputs [
    A {from=MaterialExpressionNamedRerouteUsage_6}
    B {from=MaterialExpressionNamedRerouteUsage_5}
   ]
  }
  MaterialExpressionMax_5 {
   class=MaterialExpressionMax
   inputs [
    A {from=MaterialExpressionMax_2}
    B {from=MaterialExpressionNamedRerouteUsage_7}
   ]
  }
  MaterialExpressionMin_0 {
   class=MaterialExpressionMin
   inputs [
    A {from=MaterialExpressionMaterialFunctionCall_12, output=Inner}
    B {from=MaterialExpressionSubtract_10, output=""}
   ]
   properties {}
  }
  MaterialExpressionMin_1 {
   class=MaterialExpressionMin
   inputs [
    A {from=MaterialExpressionMultiply_2, output=""}
    B {from="", output=""}
   ]
   properties {constB=0.5}
  }
  MaterialExpressionMultiply_0 {
   class=MaterialExpressionMultiply
   inputs [
    A {from=MaterialExpressionMaterialFunctionCall_9, output=Normalized UV}
    B {from=MaterialExpressionConstant2Vector_6}
   ]
   properties {constB=50}
  }
  MaterialExpressionMultiply_1 {
   class=MaterialExpressionMultiply
   inputs [
    A {from=MaterialExpressionMax_0}
    B {from=MaterialExpressionOneMinus_0}
   ]
  }
  MaterialExpressionMultiply_10 {
   class=MaterialExpressionMultiply
   inputs [
    A {from=MaterialExpressionSubtract_0}
    B {from=""}
   ]
   properties {constB=-0.9}
  }
  MaterialExpressionMultiply_13 {
   class=MaterialExpressionMultiply
   inputs [
    A {from=MaterialExpressionLinearInterpolate_0}
    B {from=""}
   ]
   properties {constB=-1}
  }
  MaterialExpressionMultiply_14 {
   class=MaterialExpressionMultiply
   inputs [
    A {from=MaterialExpressionComponentMask_9}
    B {from=""}
   ]
   properties {constB=25}
  }
  MaterialExpressionMultiply_18 {
   class=MaterialExpressionMultiply
   inputs [
    A {from=MaterialExpressionComponentMask_10}
    B {from=""}
   ]
   properties {constB=3}
  }
  MaterialExpressionMultiply_2 {
   class=MaterialExpressionMultiply
   inputs [
    A {from=MaterialExpressionMax_0}
    B {from=MaterialExpressionClamp_0}
   ]
  }
  MaterialExpressionMultiply_28 {
   class=MaterialExpressionMultiply
   inputs [
    A {from=MaterialExpressionSubtract_11}
    B {from=""}
   ]
   properties {constB=-2}
  }
  MaterialExpressionMultiply_3 {
   class=MaterialExpressionMultiply
   inputs [
    A {from=MaterialExpressionClamp_0}
    B {from=""}
   ]
   properties {constB=2}
  }
  MaterialExpressionMultiply_38 {
   class=MaterialExpressionMultiply
   inputs [
    A {from=MaterialExpressionMaterialFunctionCall_38, output=SDF}
    B {from=""}
   ]
   properties {constB=-3}
  }
  MaterialExpressionMultiply_6 {
   class=MaterialExpressionMultiply
   inputs [
    A {from=MaterialExpressionNamedRerouteUsage_8}
    B {from=MaterialExpressionMax_5}
   ]
  }
  MaterialExpressionMultiply_7 {
   class=MaterialExpressionMultiply
   inputs [
    A {from=MaterialExpressionMultiply_6}
    B {from=""}
   ]
   properties {constB=0.05}
  }
  MaterialExpressionNamedRerouteDeclaration_10 {
   class=MaterialExpressionNamedRerouteDeclaration
   inputs [None {from=MaterialExpressionMultiply_38}]
   properties {name=Mask_FXGlow, nodeColor=<rgba 0.75 0.6063 0.375 0.75>}
  }
  MaterialExpressionNamedRerouteDeclaration_7 {
   class=MaterialExpressionNamedRerouteDeclaration
   inputs [None {from=MaterialExpressionSaturate_9}]
   properties {name=MaskFXBacker, nodeColor=<rgba 0.75 0.6063 0.375 0.75>}
  }
  MaterialExpressionNamedRerouteDeclaration_8 {
   class=MaterialExpressionNamedRerouteDeclaration
   inputs [None {from=MaterialExpressionSaturate_11}]
   properties {name=Mask_StripesAccent, nodeColor=<rgba 0.75 0.6063 0.375 0.75>}
  }
  MaterialExpressionNamedRerouteDeclaration_9 {
   class=MaterialExpressionNamedRerouteDeclaration
   inputs [None {from=MaterialExpressionSaturate_10}]
   properties {name=Mask_DotGridAccent, nodeColor=<rgba 0.75 0.6063 0.375 0.75>}
  }
  MaterialExpressionNamedRerouteUsage_0 {class=MaterialExpressionNamedRerouteUsage}
  MaterialExpressionNamedRerouteUsage_2 {class=MaterialExpressionNamedRerouteUsage}
  MaterialExpressionNamedRerouteUsage_3 {class=MaterialExpressionNamedRerouteUsage}
  MaterialExpressionNamedRerouteUsage_5 {class=MaterialExpressionNamedRerouteUsage}
  MaterialExpressionNamedRerouteUsage_6 {class=MaterialExpressionNamedRerouteUsage}
  MaterialExpressionNamedRerouteUsage_7 {class=MaterialExpressionNamedRerouteUsage}
  MaterialExpressionNamedRerouteUsage_8 {class=MaterialExpressionNamedRerouteUsage}
  MaterialExpressionOneMinus_0 {class=MaterialExpressionOneMinus, inputs [None {from=MaterialExpressionClamp_0}]}
  MaterialExpressionOneMinus_1 {class=MaterialExpressionOneMinus, inputs [None {from=MaterialExpressionMultiply_1}]}
  MaterialExpressionOneMinus_4 {class=MaterialExpressionOneMinus, inputs [None {from=MaterialExpressionMaterialFunctionCall_9, output=Normalized UV}]}
  MaterialExpressionPinBase_0 {
   class=MaterialExpressionPinBase
   inputs []
   properties {reroutePins [
     "Size/Aspect Ratio (v2)" {expression=&self:MaterialExpressionReroute_12}
     "Size/Aspect Ratio (v2)" {expression=&self:MaterialExpressionReroute_13}
     UV In {expression=&self:MaterialExpressionReroute_14}
    ]}
  }
  MaterialExpressionPinBase_1 {
   class=MaterialExpressionPinBase
   inputs [UV Out {from=MaterialExpressionLinearInterpolate_1, output=""}]
   properties {reroutePins [UV Out {expression=&self:MaterialExpressionReroute_15}]}
  }
  MaterialExpressionReroute_0 {class=MaterialExpressionReroute, inputs [None {from=MaterialExpressionMin_1}]}
  MaterialExpressionReroute_1 {class=MaterialExpressionReroute, inputs [None {from=MaterialExpressionFloor_0}]}
  MaterialExpressionReroute_12 {class=MaterialExpressionReroute, inputs [None {from=MaterialExpressionConstant2Vector_9}]}
  MaterialExpressionReroute_13 {class=MaterialExpressionReroute, inputs [None {from=MaterialExpressionConstant2Vector_9}]}
  MaterialExpressionReroute_14 {class=MaterialExpressionReroute, inputs [None {from=MaterialExpressionMaterialFunctionCall_25, output=Rotated Values}]}
  MaterialExpressionReroute_15 {class=MaterialExpressionReroute, inputs [None {from=MaterialExpressionLinearInterpolate_1}]}
  MaterialExpressionReroute_6 {class=MaterialExpressionReroute, inputs [None {from=MaterialExpressionAdd_15}]}
  MaterialExpressionReroute_7 {class=MaterialExpressionReroute, inputs [None {from=MaterialExpressionAdd_15}]}
  MaterialExpressionSaturate_1 {class=MaterialExpressionSaturate, inputs [None {from=MaterialExpressionSign_0}]}
  MaterialExpressionSaturate_10 {class=MaterialExpressionSaturate, inputs [None {from=MaterialExpressionMaterialFunctionCall_44, output=Outer}]}
  MaterialExpressionSaturate_11 {class=MaterialExpressionSaturate, inputs [None {from=MaterialExpressionMin_0}]}
  MaterialExpressionSaturate_2 {class=MaterialExpressionSaturate, inputs [None {from=MaterialExpressionMaterialFunctionCall_22, output=Inner}]}
  MaterialExpressionSaturate_9 {class=MaterialExpressionSaturate, inputs [None {from=MaterialExpressionMaterialFunctionCall_37, output=Inner}]}
  MaterialExpressionSign_0 {class=MaterialExpressionSign, inputs [None {from=MaterialExpressionSine_3, output=""}], properties {}}
  MaterialExpressionSine_0 {class=MaterialExpressionSine, inputs [None {from=MaterialExpressionFrac_1}]}
  MaterialExpressionSine_1 {class=MaterialExpressionSine, inputs [None {from=MaterialExpressionFrac_2}]}
  MaterialExpressionSine_3 {class=MaterialExpressionSine, inputs [None {from=MaterialExpressionComponentMask_12}]}
  MaterialExpressionSubtract_0 {
   class=MaterialExpressionSubtract
   inputs [
    A {from=MaterialExpressionMaterialFunctionCall_0, output=R}
    B {from=""}
   ]
   properties {constB=0.9}
  }
  MaterialExpressionSubtract_10 {
   class=MaterialExpressionSubtract
   inputs [
    A {from=MaterialExpressionSaturate_2}
    B {from=MaterialExpressionSaturate_1}
   ]
  }
  MaterialExpressionSubtract_11 {
   class=MaterialExpressionSubtract
   inputs [
    A {from=MaterialExpressionMaterialFunctionCall_35, output=R}
    B {from=""}
   ]
   properties {constB=0.725}
  }
  MaterialExpressionSubtract_2 {
   class=MaterialExpressionSubtract
   inputs [
    A {from=MaterialExpressionSine_0}
    B {from=MaterialExpressionConstant_8}
   ]
  }
  MaterialExpressionSubtract_4 {
   class=MaterialExpressionSubtract
   inputs [
    A {from=MaterialExpressionMaterialFunctionCall_17, output=Rotated Values}
    B {from=MaterialExpressionConstant_12}
   ]
  }
  MaterialExpressionSubtract_9 {
   class=MaterialExpressionSubtract
   inputs [
    A {from=MaterialExpressionMaterialFunctionCall_17, output=Rotated Values}
    B {from=MaterialExpressionConstant_13}
   ]
   properties {constB=0}
  }
 }
 properties {bAutomaticallySetUsageInEditor=false, blendMode=BLEND_Translucent, materialDomain=MD_UI, usedShadingModels=Default Lit}
}
