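# Translucent UI material that draws the Hive POI entry icon: an exclamation mark on a notification triangle over a glow backer and accent.
# Icon textures are thresholded as distance fields through anti-aliasing functions, each blended with a softer-threshold shadow copy.
# Box and trapezoid SDFs combined with mirrored rotated UVs shape the glow, with a vertical gradient fading the backer.
# A LowHighSwitch parameter blends between two color palettes for the composited layers.
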Material /TheHivePOIUI/UI/Materials/M_HivePOIEnter_Icon
graph {
 properties {bAutomaticallySetUsageInEditor=false, blendMode=BLEND_Translucent, materialDomain=MD_UI, usedShadingModels=Default Lit}
 nodes [
  n1 = LinearInterpolate(
   A=Constant3Vector() {constant=<rgba 1 0.3515 0.0123 1>}
   B=Constant3Vector() {constant=<rgba 1 0.7413 0.6059 1>}
   Alpha=ScalarParameter() {parameterName=LowHighSwitch}
  )
  n2 = NamedRerouteUsage()
  n3 = Constant3Vector() {constant=<rgba 1 0.2346 0 1>}
  n4 = LinearInterpolate(B=@n3, Alpha=NamedRerouteUsage())
  n5 = LinearInterpolate(B=LinearInterpolate(A=Constant() {r=1}, B=@n3, Alpha=ScalarParameter() {parameterName=LowHighSwitch}), Alpha=NamedRerouteUsage())
  n6 = Constant3Vector() {constant=<rgba 0.4793 0.4678 0.4678 1>}
  n7 = Constant3Vector() {constant=<rgba 0.0844 0.0742 0.0704 1>}
  n8 = NamedRerouteUsage()
  n9 = Constant3Vector() {constant=<rgba 1 0.3515 0.0123 1>}
  n10 = Constant3Vector() {constant=<rgba 0.2502 0.1144 0.0369 1>}
  n11 = LinearInterpolate(
   B=LinearInterpolate(
    A=LinearInterpolate(A=@n6, B=@n7, Alpha=@n8)
    B=LinearInterpolate(A=@n9, B=@n10, Alpha=@n8)
    Alpha=ScalarParameter() {parameterName=LowHighSwitch}
   )
   Alpha=NamedRerouteUsage()
  )
  n12 = NamedRerouteUsage()
  n13 = LinearInterpolate(
   A=LinearInterpolate(A=Constant3Vector() {constant=<rgba 0.0232 0.0168 0.0152 1>}, B=Constant3Vector() {constant=<rgba 0.4342 0.4072 0.3968 1>}, Alpha=@n12)
   B=LinearInterpolate(A=Constant3Vector() {constant=<rgba 0.0865 0.0395 0.0242 1>}, B=Constant3Vector() {constant=<rgba 1 0.3515 0.0123 1>}, Alpha=@n12)
   Alpha=ScalarParameter() {parameterName=LowHighSwitch}
  )
  n14 = NamedRerouteUsage()
  n15 = LinearInterpolate(A=NamedRerouteUsage(), Alpha=NamedRerouteUsage())
  Add(
   A=Multiply(A=Multiply(A=@n1, B=@n2)) {constB=0.25}
   B=LinearInterpolate(A=Max(A=@n4, B=@n5), B=LinearInterpolate(A=@n11, B=@n13, Alpha=@n14), Alpha=@n15)
  )
  n16 = NamedRerouteUsage()
  n17 = NamedRerouteUsage()
  Multiply(A=Max(A=@n15, B=Add(A=@n16, B=@n17))) {constB=0.7}
  n18 = Constant2Vector() {g=0.27}
  n19 = Add(
   A=MaterialFunctionCall() {materialFunction=&GetUserInterfaceUV}.Normalized UV
   B=Add(A=@n18, B=Constant2Vector() {g=-0.02})
  )
  n20 = Constant() {r=0.35}
  n21 = Add(A=@n20, B=Constant() {r=-0.1})
  n22 = TextureSample(UVs=MaterialFunctionCall(UV=MaterialFunctionCall(UVs=@n19, Texture Scale=@n21) {materialFunction=&ScaleUVsByCenter}.UVs) {materialFunction=&MF_UI_UVScaling}.Scaled to Fit) {samplerType=SAMPLERTYPE_LinearColor, texture=/TheHivePOIUI/UI/Textures/T_UI_Icon_Notification_Exclamation}
  n23 = Constant() {r=0.8}
  n24 = MaterialFunctionCall(DistanceField=@n22.RGB, Threshold=@n23) {materialFunction=&MF_UI_AntiAliasing}
  n25 = MaterialFunctionCall(DistanceField=@n22.RGB, Threshold=Add(A=@n23) {constB=-0.175}) {materialFunction=&MF_UI_AntiAliasing}
  NamedRerouteDeclaration(None=LinearInterpolate(
    A=NamedRerouteDeclaration(None=@n24.Outer) {name="!icon", nodeColor=<rgba 0.75 0.6063 0.375 0.75>}
    B=NamedRerouteDeclaration(None=@n25.Outer) {name="!shadow", nodeColor=<rgba 0.75 0.6063 0.375 0.75>}
   ) {constAlpha=0.4}) {name="!Icon+Shadow", nodeColor=<rgba 0.75 0.6063 0.375 0.75>}
  n26 = Constant2Vector() {desc=IconGlowOffset, g=0.125}
  n27 = Constant2Vector() {g=-0.225}
  n28 = MaterialFunctionCall() {materialFunction=&GetUserInterfaceUV}
  n29 = ConstantBiasScale(None=Add(A=Add(A=@n26, B=@n27), B=@n28.Normalized UV)) {bias=-0.5, scale=2}
  n30 = Constant()
  n31 = Constant() {r=0.05}
  n32 = MaterialFunctionCall(UVs=Multiply(A=@n29) {constB=-1}, Rotation Center=@n30, "Rotation Angle (0-1)"=@n31) {materialFunction=&CustomRotator}
  n33 = MaterialFunctionCall(UVs=@n29, Rotation Center=@n30, "Rotation Angle (0-1)"=Multiply(A=@n31) {constB=-1}) {materialFunction=&CustomRotator}
  n34 = Add(A=Constant2Vector() {g=0.3}, B=@n26)
  n35 = Max(
   A=Max(A=ComponentMask(None=@n32.Rotated Values) {r=true}, B=ComponentMask(None=@n33.Rotated Values) {r=true})
   B=MaterialFunctionCall!(UVs=Add(A=@n34, B=@n28.Normalized UV), Size=Constant2Vector() {g=0.6, r=0.19}, CornerRoundness, KeepAspectRatio, PixelSize) {materialFunction=/FortUILibrary/MaterialFunctions/Shapes/SDFs/MF_SDF_Box}.SDF
  )
  n36 = NamedRerouteDeclaration(None=Saturate(None=MaterialFunctionCall(DistanceField=@n35) {materialFunction=&MF_UI_AntiAliasing}.Inner)) {name=Mask_GlowBacker, nodeColor=<rgba 0.75 0.6063 0.375 0.75>}
  n37 = Multiply(A=ComponentMask(None=@n28.Normalized UV) {g=true}) {constB=2}
  NamedRerouteDeclaration(None=Saturate(None=Multiply(A=@n36, B=@n37))) {name=OpacityMask_GlowBacker, nodeColor=<rgba 0.75 0.6063 0.375 0.75>}
  n38 = Add(A=MaterialFunctionCall() {materialFunction=&GetUserInterfaceUV}.Normalized UV, B=@n26)
  n39 = Constant3Vector() {constant=<rgba 0.27 0.16 0.17 0>}
  n40 = MaterialFunctionCall!(UVs=Add(A=Constant2Vector() {g=0.068}, B=@n38), TopWidth=@n39.R, BottomWidth=@n39.R, Height=@n39.R) {materialFunction=/FortUILibrary/MaterialFunctions/Shapes/SDFs/MF_SDF_IsoscelesTrapezoid}
  n41 = Constant3Vector() {constant=<rgba 0.24 0.165 0.12 0>}
  n42 = MaterialFunctionCall(UVs=Add(A=@n38, B=Constant2Vector() {g=0.03}), TopWidth=@n41.R, BottomWidth=@n41.R, Height=@n41.R) {materialFunction=&MF_SDF_IsoscelesTrapezoid}
  n43 = ComponentMask(None=@n32.Rotated Values) {g=true}
  n44 = ComponentMask(None=Multiply(A=@n33.Rotated Values) {constB=-1}) {g=true}
  n45 = Max(A=Max(A=Multiply(A=@n40.SDF) {constB=-1}, B=Multiply(A=@n42.SDF)), B=Add(A=Max(A=@n43, B=@n44)) {constB=-0.58})
  NamedRerouteDeclaration(None=Saturate(None=MaterialFunctionCall(DistanceField=@n45) {materialFunction=&MF_UI_AntiAliasing}.Inner)) {name=Mask_GlowAccent, nodeColor=<rgba 0.75 0.6063 0.375 0.75>}
  n46 = Add(A=MaterialFunctionCall() {materialFunction=&GetUserInterfaceUV}.Normalized UV, B=@n18)
  n47 = TextureSample(UVs=MaterialFunctionCall(UV=MaterialFunctionCall(UVs=@n46, Texture Scale=@n20) {materialFunction=&ScaleUVsByCenter}.UVs) {materialFunction=&MF_UI_UVScaling}.Scaled to Fit) {samplerType=SAMPLERTYPE_LinearColor, texture=/TheHivePOIUI/UI/Textures/T_UI_Icon_Notification_Bg}
  n48 = Constant() {r=0.75}
  n49 = MaterialFunctionCall(DistanceField=@n47.RGB, Threshold=@n48) {materialFunction=&MF_UI_AntiAliasing}
  n50 = MaterialFunctionCall(DistanceField=@n47.RGB, Threshold=Add(A=@n48) {constB=-0.1}) {materialFunction=&MF_UI_AntiAliasing}
  NamedRerouteDeclaration(None=LinearInterpolate(
    A=NamedRerouteDeclaration(None=@n49.Outer) {name=TriangleIcon, nodeColor=<rgba 0.75 0.6063 0.375 0.75>}
    B=NamedRerouteDeclaration(None=@n50.Outer) {name=TriangleShadow, nodeColor=<rgba 0.75 0.6063 0.375 0.75>}
   ) {constAlpha=0.4}) {name=TriangleIcon+Shadow, nodeColor=<rgba 0.75 0.6063 0.375 0.75>}
 ]
}
